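# Footprint material for the Cattus Head enemy, shading an imprinted mound in the ground with Winter and non-winter variants.
# Mound mask channels, height lerps and vertex color layer puddle, mud, gravel or grass, and concrete.
# A Winter static switch, on by default, selects snow-and-ice shading; otherwise puddles get panning noise normals and RVT blending applies.
# A baked mound normal is layered in, with cheaper mobile shading paths.
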Material /Figment_S09/Characters/Enemies/Cattus_Head/Materials/M_Footprint
graph {
 properties {bTangentSpaceNormal=false, bUseMaterialAttributes=true, pixelDepthOffsetMode=PDOM_Legacy, usedShadingModels=Default Lit}
 nodes [
  Constant3Vector()
  n1 = MaterialFunctionCall!(In=NamedRerouteUsage(), Color=VectorParameter() {parameterName=GravelTint, defaultValue=<rgba 1 1 1 1>}.RGB) {materialFunction=/FigmentEnvironment/LM_8_1_1/Athena/Environments/Landscape/MaterialFunctions/MF_Athena_Material_BaseMultiply}
  n2 = MaterialFunctionCall() {materialFunction=/Game/Athena/Environments/Landscape/MaterialFunctions/Standard/MF_Athena_Grass_01}
  n3 = TextureSampleParameter2D() {parameterName=Mask, samplerType=SAMPLERTYPE_Masks, texture=/Figment_S09/Characters/Enemies/Cattus_Head/Textures/S_Cattus_Mound_D}
  n4 = ScalarParameter() {defaultValue=1, parameterName=R_MaskContrast}
  n5 = Saturate(None=Add(A=Power(Base=@n3.R, Exp=@n4), B=ScalarParameter() {parameterName=R_MaskOffset}))
  n6 = Multiply(A=TextureCoordinate(), B=ScalarParameter() {defaultValue=7, group=Mud, parameterName=Mud Scale})
  n7 = TextureSample(UVs=@n6) {texture=/Game/Athena/Environments/Island/Textures/Dirt_D}
  n8 = ScalarParameter() {defaultValue=0.35, group=Gravel, parameterName=Mud Height}
  n9 = MaterialFunctionCall(
   Transition Phase=ScalarParameter() {defaultValue=0.499, group=Gravel, parameterName=GravelMud/Grass Trans}
   Height Texture=Subtract(A=@n5, B=Multiply(A=@n7.A, B=@n8))
   Contrast=ScalarParameter() {defaultValue=2, group=Gravel, parameterName=GravelMud/Grass Contrast}
  ) {materialFunction=&HeightLerp}
  n10 = Multiply(A=@n7.RGB)
  n11 = Fresnel() {exponent=3}
  n12 = MakeMaterialAttributes!(
   BaseColor=Multiply(A=Multiply(A=@n10) {constB=0.5}, B=OneMinus(None=@n11))
   Metallic
   Specular=ScalarParameter() {defaultValue=0.1, group=Puddle, parameterName=Puddle Spec}
   Roughness=ScalarParameter() {defaultValue=0.0348, group=Puddle, parameterName=Puddle Rough}
   Anisotropy
   EmissiveColor
   Opacity
   OpacityMask
   Normal
   Tangent
   WorldPositionOffset
   SubsurfaceColor
   ClearCoat
   ClearCoatRoughness
   AmbientOcclusion
   Refraction
   CustomizedUVs_0
   CustomizedUVs_1
   CustomizedUVs_2
   CustomizedUVs_3
   CustomizedUVs_4
   CustomizedUVs_5
   CustomizedUVs_6
   CustomizedUVs_7
   PixelDepthOffset
   ShadingModel
   Displacement
  )
  n13 = TextureSampleParameter2D() {parameterName=Mask, samplerType=SAMPLERTYPE_Masks, texture=&S_Cattus_Mound_D}
  n14 = MakeMaterialAttributes(
   BaseColor=@n10
   Specular=LinearInterpolate(
    A=ScalarParameter() {defaultValue=0.2, group=Mud, parameterName=SpecMud}
    B=ScalarParameter() {defaultValue=0.5, group=Mud, parameterName=Spec Mud Top}
    Alpha=@n13.G
   )
   Roughness=LinearInterpolate(
    A=ScalarParameter() {defaultValue=0.3, group=Mud, parameterName=Mud R Bott}
    B=ScalarParameter() {defaultValue=0.85, group=Mud, parameterName=Mud R Top}
    Alpha=@n13.R
   )
   Normal=MaterialFunctionCall(
    Normal=TextureSample(UVs=@n6) {samplerType=SAMPLERTYPE_Normal, texture=/Game/Packages/DS_Fortnight/ExperimentalTextures/Dirt_N}.RGB
    Flatness=ScalarParameter() {defaultValue=0.1614, group=Mud, parameterName=Mud Flatten}
   ) {materialFunction=&FlattenNormal}.Result
  )
  n15 = Multiply(A=TextureCoordinate(), B=ScalarParameter() {defaultValue=4.8933, group=Gravel, parameterName=MudTextureTiling})
  n16 = TextureSample(UVs=@n15) {texture=/Game/Packages/DS_Fortnight/WhiteRockCustom_D}
  n17 = Saturate(None=LinearInterpolate(A=Constant() {r=-1}, B=Constant() {r=1}, Alpha=@n16.A))
  n18 = SetMaterialAttributes!(
   MaterialAttributes
   Base Color=@n16.RGB
   Specular=LinearInterpolate(A=Constant() {r=0.05}, B=Constant() {r=0.5}, Alpha=@n17) {constB=0.4}
   Roughness=LinearInterpolate(A=Constant() {r=0.99}, B=Constant() {r=0.65}, Alpha=@n17)
   Normal=TextureSample(UVs=@n15) {samplerType=SAMPLERTYPE_Normal, texture=/Game/Packages/DS_Fortnight/WhiteRockCustom_N}.RGB
  ) {attributeSetTypes [
    #1
    #2
    #3
    #4
   ]}
  n19 = ScalarParameter() {defaultValue=0.419, group=Gravel, parameterName=Gravel/Mud Transition}
  n20 = TextureSampleParameter2D() {parameterName=Mask, samplerType=SAMPLERTYPE_Masks, texture=&S_Cattus_Mound_D}
  n21 = Add(A=@n20.R, B=Multiply(A=@n20.R, B=@n16.A))
  n22 = ScalarParameter() {defaultValue=1, group=Gravel, parameterName=GravelMud Contrast}
  n23 = Saturate(None=Subtract(
    A=MaterialFunctionCall(Transition Phase=@n19, Height Texture=@n21, Contrast=@n22) {materialFunction=&HeightLerp}.Alpha
    B=VertexColor().R
   ))
  n24 = MaterialFunctionCall(Transition Phase=Constant() {r=0.65}, Height Texture=@n13.R, Contrast=Constant() {r=2}) {materialFunction=&HeightLerp}
  n25 = GetMaterialAttributes(None=MaterialFunctionCall!(
    Base Material=@n12
    Top Material=MaterialFunctionCall(Base Material=@n14, Top Material=@n18, Alpha=@n23) {materialFunction=/Engine/Functions/MaterialLayerFunctions/MatLayerBlend_Standard}.Blended Material
    Alpha=@n24.Alpha
   ) {materialFunction=&MatLayerBlend_Standard}.Blended Material) {attributeGetTypes [#1]}
  n26 = TextureSampleParameter2D() {parameterName=Mask, samplerType=SAMPLERTYPE_Masks, texture=&S_Cattus_Mound_D}
  n27 = Multiply(
   A=@n25.Base Color
   B=LinearInterpolate(
    A=VectorParameter() {group=Mud, parameterName=Mud Tint Low, defaultValue=<rgba 0.1146 0.098 0.0525 1>}.RGB
    B=VectorParameter() {group=Mud, parameterName=Mut Tint Top, defaultValue=<rgba 0.6979 0.583 0.4358 1>}.RGB
    Alpha=@n26.R
   )
  )
  n28 = GetMaterialAttributes(None=MaterialFunctionCall!(Concrete diffuse, Concrete Spec, Concrete Normal) {materialFunction=/Game/Packages/DS_Fortnight/SM/Materials/concreteMF}.Material Attribute) {attributeGetTypes [#1, #2]}
  n29 = ScalarParameter() {parameterName=ConcreteSpecMin}
  n30 = ScalarParameter() {defaultValue=1, parameterName=ConcreteSpecMax}
  n31 = MaterialFunctionCall(
   Base Material=MaterialFunctionCall(
    Base Material=SetMaterialAttributes!(MaterialAttributes=@n25.MaterialAttributes, Base Color=@n27) {attributeSetTypes [#1]}
    Top Material=StaticSwitchParameter(True=@n1.Result, False=@n2.Result) {defaultValue=true, parameterName=UseGravelNotGrass}
    Alpha=@n9.Alpha
   ) {materialFunction=&MatLayerBlend_Standard}.Blended Material
   Top Material=SetMaterialAttributes!(
    MaterialAttributes=@n28.MaterialAttributes
    Base Color=@n28.MaterialAttributes
    Specular=LinearInterpolate(A=@n29, B=@n30, Alpha=@n28.Specular)
   ) {attributeSetTypes [#1, #2]}
   Alpha=MaterialFunctionCall(Transition Phase=VertexColor().R, Height Texture=@n9.Alpha) {materialFunction=&HeightLerp}.Alpha
  ) {materialFunction=&MatLayerBlend_Standard}
  n32 = GetMaterialAttributes(None=@n31.Blended Material) {attributeGetTypes [#1, #3]}
  n33 = ComponentMask(None=@n32.Base Color) {r=true}
  n34 = NamedRerouteUsage()
  MaterialFunctionCall!(
   "Xb1,PS4"=@n31.Blended Material
   "PC-High&&Epic,XSX,PS5"=@n31.Blended Material
   "PC-Perf,Android"=SetMaterialAttributes!(
    MaterialAttributes=@n32.MaterialAttributes
    Specular=LinearInterpolate(B=@n33, Alpha=@n34) {constA=1, constB=0.5}
    Roughness=LinearInterpolate(B=@n32.Roughness, Alpha=@n34) {constB=0.5}
   ) {attributeSetTypes [#2, #3]}
   Switch=@n31.Blended Material
   "PC-Low(NotPerf)&&Medium"=@n31.Blended Material
  ) {materialFunction=/Game/Effects/Niagara/MasterMaterials/Functions/Common/PlatformSwitch_MA}
  MaterialFunctionCall!(Snow) {materialFunction=/FigmentEnvironment/LM_8_1_1/Athena/Environments/Landscape/MaterialFunctions/Snow/MF_Figment_BlendSnowBiomeColor}
  n35 = MaterialFunctionCall!(Material) {materialFunction=/FigmentEnvironment/LM_8_1_1/Athena/Environments/Materials/MF_StaticMeshBiomeColor}
  n36 = MaterialFunctionCall(
   In=MaterialFunctionCall!(UseLandscapeCoordsOnSwitch) {materialFunction=/FigmentEnvironment/LM_8_1_1/Athena/Environments/Landscape/MaterialFunctions/Snow/MF_Athena_Snow_01}.Material
   Color=@n35.Color
  ) {materialFunction=&MF_Athena_Material_BaseMultiply}
  NamedRerouteDeclaration(None=@n36.Result) {name=Snow, nodeColor=<rgba 0.2656 0 1 1>}
  NamedRerouteDeclaration(None=@n35.Color) {name=ColorMap, nodeColor=<rgba 0.7578 0 1 1>}
  NamedRerouteDeclaration(None=@n18) {name=Gravel, nodeColor=<rgba 0.0547 0 1 1>}
  NamedRerouteDeclaration(None=@n14) {name=Mud, nodeColor=<rgba 0.2656 1 0 1>}
  NamedRerouteDeclaration(None=@n12) {name=Puddle, nodeColor=<rgba 1 0 0.5859 1>}
  NamedRerouteDeclaration(None=@n26.RGB) {name=Mask, nodeColor=<rgba 1 0 0.9141 1>}
  n37 = VectorParameter() {parameterName=IceColor, defaultValue=<rgba 0.0762 0.1743 0.2812 1>}
  n38 = ScalarParameter() {defaultValue=0.2, parameterName=IceRaughness}
  n39 = GetMaterialAttributes(None=@n36.Result) {attributeGetTypes [#1, #3, #2]}
  n40 = LinearInterpolate(A=ScalarParameter() {defaultValue=0.5, parameterName=R_MaskMin}, Alpha=@n5) {constA=0.5}
  n41 = Multiply(A=@n39.Base Color, B=@n40)
  n42 = Multiply(A=@n39.Roughness, B=@n40)
  n43 = Multiply(A=@n39.Specular, B=@n5)
  NamedRerouteDeclaration(None=MaterialFunctionCall(
    Base Material=MakeMaterialAttributes(BaseColor=@n37.RGB, Roughness=@n38)
    Top Material=SetMaterialAttributes!(MaterialAttributes=@n39.MaterialAttributes, Base Color=@n41, Roughness=@n42, Specular=@n43) {attributeSetTypes [#1, #3, #2]}
    Alpha=@n5
   ) {materialFunction=&MatLayerBlend_Standard}.Blended Material) {name=WinterMat, nodeColor=<rgba 0.8281 0 1 1>}
  NamedRerouteDeclaration(None=@n24.Alpha) {name=PuddleMask, nodeColor=<rgba 1 0.0938 0 1>}
  Power(Base=ComponentMask(None=NamedRerouteUsage()) {g=true}) {constExponent=0.75}
  n44 = TextureSampleParameter2D() {parameterName=Mask, samplerType=SAMPLERTYPE_Masks, texture=&S_Cattus_Mound_D}
  Saturate(None=@n44.B)
  n45 = StaticSwitchParameter(True=NamedRerouteUsage(), False=@n31.Blended Material) {defaultValue=true, group=Biome, parameterName=Winter}
  n46 = GetMaterialAttributes(None=@n45) {attributeGetTypes [#1, #2]}
  n47 = LinearInterpolate(A=VectorParameter() {parameterName=PuddleTintLow, defaultValue=<rgba 0 0 0 1>}.RGB, Alpha=NamedRerouteUsage()) {constA=0.1, constB=0}
  n48 = ComponentMask(None=NamedRerouteUsage()) {r=true}
  n49 = SetMaterialAttributes!(
   MaterialAttributes
   Base Color=Saturate(None=Add(A=@n46.Base Color, B=@n47))
   Specular=LinearInterpolate(B=ComponentMask(None=@n46.Base Color) {r=true}, Alpha=Power(Base=@n48) {constExponent=0.75}) {constA=0.2, constB=0.5}
   Roughness=LinearInterpolate(Alpha=NamedRerouteUsage()) {constA=0.2}
  ) {attributeSetTypes [#1, #2, #3]}
  n50 = TextureSampleParameter2D() {parameterName=Normal, samplerType=SAMPLERTYPE_Normal, texture=/Figment_S09/Characters/Enemies/Cattus_Head/Textures/S_Cattus_Mound_N}
  n51 = Divide(A=ComponentMask(None=WorldPosition().XYZ) {g=true, r=true}, B=Constant() {r=250})
  n52 = TextureSample(UVs=Panner(Coordinate=@n51) {speedX=0.1, speedY=0.1}) {samplerSource=SSM_Wrap_WorldGroupSettings, samplerType=SAMPLERTYPE_Normal, texture=/Figment_S09/Environments/Foliage/Textures/T_SwirlNoise_N}
  n53 = TextureSample(UVs=Panner(Coordinate=@n51) {speedX=-0.1, speedY=-0.1}) {samplerSource=SSM_Wrap_WorldGroupSettings, samplerType=SAMPLERTYPE_Normal, texture=&T_SwirlNoise_N}
  n54 = Add(A=@n52.RGB, B=Multiply(A=@n53.RGB, B=Constant3Vector() {constant=<rgba 1 1 0 0>}))
  n55 = Constant() {r=0.99}
  n56 = ShadingPathSwitch(
   Default=LinearInterpolate(
    A=MaterialFunctionCall(Normal=@n54, Flatness=@n55) {materialFunction=&FlattenNormal}.Result
    B=@n50.RGB
    Alpha=@n24.Alpha
   )
   Forward=LinearInterpolate(A=Constant3Vector() {constant=<rgba 0 0 1 0>}, B=@n50.RGB, Alpha=@n24.Alpha)
   Mobile=@n50.RGB
  )
  n57 = GetMaterialAttributes(None=MaterialFunctionCall!(
    Material=ShadingPathSwitch(Default=@n45, Deferred=@n45, Forward=@n45, Mobile=@n49)
    Normal=StaticSwitchParameter(True=@n50.RGB, False=@n56) {defaultValue=true, group=Biome, parameterName=Winter}
   ) {materialFunction=/Engine/Functions/MaterialLayerFunctions/MatLayerBlend_BakedNormal}.Blended Material) {attributeGetTypes [#1]}
  n58 = GetMaterialAttributes(None=SetMaterialAttributes!(
    MaterialAttributes=@n57.MaterialAttributes
    Base Color=MaterialFunctionCall!(BaseColor=@n57.Base Color) {materialFunction=/Game/Athena/Environments/Landscape/MaterialFunctions/Standard/MF_Athena_GrassMatch}.Result
   ) {attributeSetTypes [#1]}) {attributeGetTypes [#4]}
  n59 = SetMaterialAttributes!(MaterialAttributes=@n58.MaterialAttributes, Normal=Transform(None=@n58.Normal)) {attributeSetTypes [#4]}
  n60 = SmoothStep(Min=ScalarParameter() {defaultValue=0.6, parameterName=MaskMin}, Max=ScalarParameter() {defaultValue=1, parameterName=MaskMax}, Value=@n44.R)
  n61 = Power(Base=Saturate(None=OneMinus(None=@n44.G))) {constExponent=10}
  StaticSwitchParameter(
   True=BlendMaterialAttributes(A=@n59, B=NamedRerouteUsage(), Alpha=Subtract(A=@n60, B=@n61))
   False=MaterialFunctionCall!(In=@n59) {materialFunction=/FigmentCore/Game/TechArt/Material_Functions/MF_Figment_RVTBlendingOn}.Result
  ) {defaultValue=true, group=Biome, parameterName=Winter}
 ]
}
